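# Niagara system for the Dark Magic car boost cosmetic, attaching glows, a mesh and layered trails to the vehicle's boost sockets.
# A sprite/mesh emitter bursts particles by spawn group, partitioned across ports on the skeletal mesh.
# Two ribbon emitters trail each port, fading inherited car motion and driving per-renderer width, color and material parameters with curves.
# Painted-variant overrides recolor layers, while user car offset and, for one ribbon, game mode adjust placement.
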NiagaraSystem /VehicleCosmetics/Boosts/Boost_DarkMagic/FX/NS_Booster_DarkMagic
dependencies {
 usedDataInterfaces [
  &NiagaraDataInterfaceCurve
  &NiagaraDataInterfaceArrayFloat
  &NiagaraDataInterfaceColorCurve
  &NiagaraDataInterfaceArrayFloat3
  &NiagaraDataInterfaceSkeletalMesh
 ]
 usedDynamicInputs []
 usedModules [
  &self:SystemSpawnScript.SystemScriptSource.SystemScriptGraph.NiagaraNodeAssignment_0.SetVariables_#1
  &SystemState
 ]
}
emitters {
 NE_VC_Boost_Sprite_Parent {
  emitterUpdateScript [
   &EmitterState(
    Inactive Response=::NewEnumerator1
    Life Cycle Mode=::NewEnumerator1
    Loop Behavior=::NewEnumerator1
    Loop Duration Mode=::NewEnumerator1
    UseLoopDelay=false
   )
   &SpawnBurst_Instantaneous(Spawn Count=2, Spawn Group=4, Use Spawn Probability=false, UseLoopCountLimit=false) {module=SpawnBurst_Instantaneous003}
   &VC_FilteredSocketCount()
   &SpawnBurst_Instantaneous(Spawn Count=2, Use Spawn Probability=false, UseLoopCountLimit=false)
   &SpawnBurst_Instantaneous(Spawn Count=2, Spawn Group=1, Use Spawn Probability=false, UseLoopCountLimit=false) {module=SpawnBurst_Instantaneous001}
   &SpawnBurst_Instantaneous(Spawn Count=2, Spawn Group=2, Use Spawn Probability=false, UseLoopCountLimit=false) {module=SpawnBurst_Instantaneous002}
   &SpawnRate(SpawnRate=50, Use Spawn Probability=false) {enabled=false}
  ]
  particleSpawnScript [
   &VC_PaintedVariantsOverride(
    DefaultColor=<rgba 1 0.4 0 0.75>
    R1 DefaultColor=<rgba 1 0.4 0 0.5>
    R2 DefaultColor:LinearColor=<rgba 1 0.4 0 1>
    R3 DefaultColor:LinearColor=<rgba 0 0 0 10>
    Use R1 Painted Overrides=true
    Use R2 Painted Overrides=true
    Use R3 Painted Overrides=true
   )
   /SetVariables(Particles.VisibilityTag=$Emitter.SpawnGroup) {isSetParameters=true}
   &InitializeParticle(
    Color Mode=::NewEnumerator0
    Lifetime=$Emitter.CurrentLoopDuration
    Mesh Scale=<0.17 1 0.17>
    Mesh Scale Mode=::NewEnumerator4
    Sprite Size=<235 125>
    Sprite Size Mode=::NewEnumerator4
    UsePositionOffset=false
    Write Material Random=true
   )
   &PartitionParticles(Number Of Partitions=$Emitter.PortCount)
   &SkeletalMeshLocation(
    Bone / Socket Index=$Particles.PartitionParticles.Partition
    Bone / Socket Sampling Mode=::NewEnumerator5
    MaskSpawnGroup=false
    Mesh Sampling Type=::NewEnumerator6
    Sampled Position Offset=&AddVectorToPosition(
     Position=&MergeHelper_VectorSwitch(0=<-70 0 0>, 1=<-65 0 0>, 2=<-100 0 0>, 3=<-100 0 0>, 4=<-65 0 0>, Index=$Emitter.SpawnGroup, Num=::NewEnumerator3)
     Vector=$User.DTCarOffset
    )
    Skeletal Mesh=$System.SkeletalMesh
   )
   &InitialMeshOrientation(
    Mesh Axis To Orient=&SelectVectorFromArray(
     Direct Array Index=$Particles.PartitionParticles.Partition
     Vector Selection Array=DataInterface() {propertyValues {FloatData [<-0.15 -2 3.1147>, <1 0 0>]}}
    )
    Mesh Orientation Mode=::NewEnumerator2
    Orientation Vector=&SelectVectorFromArray(
     Direct Array Index=$Particles.PartitionParticles.Partition
     Vector Selection Array=DataInterface() {propertyValues {FloatData [<0 0 1>, <1 0 0>]}}
    )
    Rotation=&SelectVectorFromArray(
     Direct Array Index=$Particles.PartitionParticles.Partition
     Vector Selection Array=DataInterface() {propertyValues {FloatData [<-1 0 0>, <1 0 0>]}}
    )
    Rotation Coordinate Space=::NewEnumerator0
    Use Rotation Vector=true
   )
   &AddVelocity(Velocity=<-1500 0 0>) {enabled=false}
   /SetVariables(Particles.FacingHorizontalGlow:Vector3f=<1 0 0>, Particles.SizeGlowPoint:Vector2f=<40 40>, Particles.SizeHorizontalGlow:Vector2f=<20 150>) {isSetParameters=true}
   &SpriteFacingAndAlignment(
    Sprite Facing=&MergeHelper_VectorSwitch(0=<0 0 0>, 1=<0 0 0>, 2=<1 1 1>, 3=?, Index=&SpawnGroup(), Num=::NewEnumerator1)
    WriteAlignment=false
    WriteFacing=true
   ) {enabled=false}
  ]
  particleUpdateScript [
   &ParticleState(Kill Particles When Lifetime Has Elapsed=false)
   &MeshRotationForce(Rotation=&SelectVectorFromArray(
     Array Sampling Mode=::NewEnumerator0
     Direct Array Index=?
     Vector Selection Array=DataInterface() {propertyValues {FloatData [<5 0 0>, <-5 0 0>]}}
     Random Seed=0
    )) {enabled=false}
   &SolveForcesAndVelocity(Clamp Velocity=false, Limit Acceleration=false)
  ]
  rendererClasses [&NiagaraSpriteRendererProperties, &NiagaraMeshRendererProperties]
  renderers [
   {rendererClass=&NiagaraSpriteRendererProperties}
   {rendererClass=&NiagaraSpriteRendererProperties, rendererIndex=1}
   {rendererClass=&NiagaraMeshRendererProperties, rendererIndex=2}
   {rendererClass=&NiagaraSpriteRendererProperties, rendererIndex=3}
  ]
 }
 Ribbon_TestSharedMaterials {
  emitterUpdateScript [
   &EmitterState()
   &VC_FilteredSocketCount()
   &VC_RateQualityHandler()
   &VC_SpawnRibbon_v1(Number of Ribbons=$Emitter.PortCount, SpawnRate=$Emitter.Spawn Rate)
  ]
  particleSpawnScript [
   &VC_PaintedVariantsOverride(
    DefaultColor=<rgba 0 0 0 4>
    R1 DefaultColor=<rgba 0.0625 0 0.25 1>
    R2 DefaultColor:LinearColor=<rgba 1 0 0 1>
    Use R1 Painted Overrides=true
    Use R2 Painted Overrides=true
   )
   &VC_InitializeRibbon_v1(
    Color=$Particles.Color
    Custom Offset=true
    Enable Skeletal Position=true
    Lifetime=0.8
    Local Vector Offset=&AddVectorToPosition(Position=<-40 0 0>, Vector=$User.DTCarOffset)
    Offset Amount=1
    Ribbon Width=160
   )
   &AddVelocity(Velocity=<2250 0 0>, Velocity Speed Scale=-1) {module=AddVelocity001}
   /SetVariables(Particles.Linear Color=&Multiply_LinearColorByFloat(Linear Color=<rgba 0.0375 0 0.1 1>, Float=1.5)) {isSetParameters=true}
   /SetVariables(Particles.SizePurple=175) {isSetParameters=true}
  ]
  particleUpdateScript [
   &ParticleState()
   &InheritSourceMovement(
    Applied Position Scale=&Multiply_VectorByFloat(
     Vector=<1 1 1>
     Float=&FloatFromCurve(FloatCurve=DataInterface() {propertyValues {Curve {keys [{interpMode=RCIM_Cubic, tangentMode=RCTM_Break, time=0.3, value=1}, {interpMode=RCIM_Cubic, tangentMode=RCTM_Break, time=1, value=0}]}}})
    )
    ApplyForce=false
    ApplyPosition=true
    ApplyPositionFalloff=false
   )
   &SolveForcesAndVelocity(Clamp Velocity=false, Limit Acceleration=false)
   &VC_UpdateRibbon_v1(Enable Skeletal Position=true)
   &ScaleRibbonWidth(Ribbon Width Scale=&FloatFromCurve(
     FloatCurve=DataInterface() {propertyValues {Curve {keys [
         {interpMode=RCIM_Cubic, tangentMode=RCTM_Break, time=0, value=0.2}
         {arriveTangent=0.0553, interpMode=RCIM_Cubic, leaveTangent=0.0553, tangentMode=RCTM_User, time=0.52, value=0.48}
         <at 1.005 0.5 RCIM_Cubic>
        ]}}}
     Scale Curve=&SelectFloatFromArray(Direct Array Index=&SpawnGroup(), Float Selection Array=DataInterface() {propertyValues {FloatData [1,-1]}})
    ))
   &ScaleColor(
    Scale Alpha=&FloatFromCurve(
     FloatCurve=DataInterface() {propertyValues {Curve {
        defaultValue=1
        keys [
         {interpMode=RCIM_Cubic, tangentMode=RCTM_User, time=0, value=0}
         {arriveTangent=3.1834, interpMode=RCIM_Cubic, leaveTangent=3.1834, time=0.2, value=1}
         {arriveTangent=0.1377, interpMode=RCIM_Cubic, leaveTangent=0.1377, tangentMode=RCTM_User, time=0.3484, value=1.1091}
         {arriveTangent=-3.1544, interpMode=RCIM_Cubic, leaveTangent=-3.1544, time=0.6544, value=0.5638}
         {arriveTangent=-4.3091, interpMode=RCIM_Cubic, tangentMode=RCTM_Break, time=0.7, value=0}
        ]
       }}}
     Scale Curve=4
    )
    ScaleA=true
    ScaleRGB=true
   )
   /SetVariables(
    Particles.MultPurple:LinearColor=&MultiplyLinearColor(
     Linear Color=$Particles.Initial.R1 Color
     Scale Factor=&MakeVector4FromVectorAndFloat(
      "Vector (RGB)"=&VectorFromFloat(Value=2)
      "Float (Alpha)"=&FloatFromCurve(
       FloatCurve=DataInterface() {propertyValues {Curve {
          defaultValue=1
          keys [
           {interpMode=RCIM_Cubic, tangentMode=RCTM_User, time=0, value=0}
           {arriveTangent=0.1377, interpMode=RCIM_Cubic, leaveTangent=0.1377, tangentMode=RCTM_User, time=0.3, value=1}
           {arriveTangent=-2.6316, interpMode=RCIM_Cubic, leaveTangent=-2.6316, time=0.5812, value=0.8}
           {arriveTangent=-4.3091, interpMode=RCIM_Cubic, tangentMode=RCTM_Break, time=0.68, value=0}
          ]
         }}}
       Scale Curve=2
      )
     )
    )
    Particles.MultREDColor:LinearColor=&MultiplyLinearColor(
     Linear Color=$Particles.Initial.R2 Color
     Scale Factor=&MakeVector4FromVectorAndFloat(
      "Vector (RGB)"=&VectorFromFloat(Value=2)
      "Float (Alpha)"=&FloatFromCurve(
       FloatCurve=DataInterface() {propertyValues {Curve {
          defaultValue=1
          keys [
           {interpMode=RCIM_Cubic, tangentMode=RCTM_User, time=0.05, value=0}
           <at 0.08 0.993 RCIM_Cubic>
           {interpMode=RCIM_Cubic, leaveTangent=-0.4615, time=0.5, value=0}
          ]
         }}}
       Scale Curve=5
      )
     )
    )
    Particles.SizPurpCale:NiagaraFloat=&Multiply_Float(
     A=$Particles.SizePurple
     B=&FloatFromCurve(
      FloatCurve=DataInterface() {propertyValues {Curve {keys [
          {interpMode=RCIM_Cubic, leaveTangent=1.8318, tangentMode=RCTM_Break, time=0, value=0.1}
          {arriveTangent=0.12, interpMode=RCIM_Cubic, leaveTangent=0.12, time=0.25, value=0.25}
          {interpMode=RCIM_Cubic, tangentMode=RCTM_Break, time=1, value=0.22}
         ]}}}
      Scale Curve=&SelectFloatFromArray(
       Direct Array Index=&SpawnGroup()
       Float Selection Array=DataInterface() {propertyValues {FloatData [1.25,-1.25]}}
      )
     )
    )
    Particles.SizeRedScale:NiagaraFloat=&Multiply_Float(
     A=$Particles.SizePurple
     B=&FloatFromCurve(
      FloatCurve=DataInterface() {propertyValues {Curve {keys [
          {interpMode=RCIM_Cubic, leaveTangent=2.8966, tangentMode=RCTM_Break, time=0, value=0}
          {arriveTangent=0.0281, interpMode=RCIM_Cubic, leaveTangent=0.5334, tangentMode=RCTM_Break, time=0.3, value=0.27}
          {interpMode=RCIM_Cubic, tangentMode=RCTM_Break, time=1, value=0.3}
         ]}}}
      Scale Curve=0.85
     )
    )
   ) {isSetParameters=true}
  ]
  rendererClasses [&NiagaraRibbonRendererProperties]
  renderers [
   {rendererClass=&NiagaraRibbonRendererProperties}
   {rendererClass=&NiagaraRibbonRendererProperties, rendererIndex=1}
   {rendererClass=&NiagaraRibbonRendererProperties, rendererIndex=2}
  ]
 }
 Ribbon_TestSharedMaterials001 {
  emitterUpdateScript [
   &EmitterState()
   &VC_FilteredSocketCount()
   &VC_RateQualityHandler(Default Rate=80)
   &VC_SpawnRibbon_v1(Number of Ribbons=$Emitter.PortCount, SpawnRate=$Emitter.Spawn Rate)
  ]
  particleSpawnScript [
   &VC_PaintedVariantsOverride(
    DefaultColor=<rgba 0 0 0 3>
    R3 DefaultColor:LinearColor=<rgba 5 0 0 0.25>
    Use R1 Painted Overrides=true
    Use R2 Painted Overrides=true
    Use R3 Painted Overrides=true
   )
   &VC_InitializeRibbon_v1(
    Color=$Particles.Color
    Custom Offset=true
    Enable Skeletal Position=true
    Lifetime=0.8
    Local Vector Offset=&AddVectorToPosition(
     Position=&MergeHelper_VectorSwitch(0=<-50 0 0>, 1=<0 0 0>, 2=<170 0 0>, 3=?, Index=$User.GameModeID, Num=::NewEnumerator1)
     Vector=$User.DTCarOffset
    )
    Offset Amount=1
    Ribbon Width=220
   )
   &AddVelocity(Velocity=<2250 0 0>, Velocity Speed Scale=-1) {module=AddVelocity001}
   /SetVariables(Particles.Size_BrokenTrail=80, Particles.Size_GradientTrail=170) {isSetParameters=true}
  ]
  particleUpdateScript [
   &ParticleState()
   &InheritSourceMovement(
    Applied Position Scale=&Multiply_VectorByFloat(
     Vector=<1 1 1>
     Float=&FloatFromCurve(FloatCurve=DataInterface() {propertyValues {Curve {keys [{interpMode=RCIM_Cubic, tangentMode=RCTM_Break, time=0.3, value=1}, {interpMode=RCIM_Cubic, tangentMode=RCTM_Break, time=1, value=0}]}}})
    )
    ApplyForce=false
    ApplyPosition=true
    ApplyPositionFalloff=false
   ) {enabled=false}
   &InheritSourceMovement(
    Applied Position Scale=&Multiply_VectorByFloat(
     Vector=<1 1 1>
     Float=&FloatFromCurve(FloatCurve=DataInterface() {propertyValues {Curve {
         defaultValue=1
         keys [{interpMode=RCIM_Cubic, tangentMode=RCTM_User, time=0.3, value=1}, {interpMode=RCIM_Cubic, tangentMode=RCTM_User, time=1, value=0}]
        }}})
    )
    ApplyForce=false
    ApplyPosition=true
    ApplyPositionFalloff=false
   ) {module=InheritSourceMovement001}
   &SolveForcesAndVelocity(Clamp Velocity=false, Limit Acceleration=false)
   &VC_UpdateRibbon_v1(Enable Skeletal Position=true)
   &ScaleRibbonWidth(Ribbon Width Scale=&FloatFromCurve(
     FloatCurve=DataInterface() {propertyValues {Curve {keys [
         {interpMode=RCIM_Cubic, leaveTangent=3.3191, tangentMode=RCTM_Break, time=0, value=0}
         {arriveTangent=0.816, interpMode=RCIM_Cubic, leaveTangent=0.4562, tangentMode=RCTM_Break, time=0.1, value=0.15}
         {arriveTangent=-0.0351, interpMode=RCIM_Cubic, leaveTangent=0.2983, tangentMode=RCTM_Break, time=0.46, value=0.28}
         <at 1.0031 0.35 RCIM_Cubic>
        ]}}}
     Scale Curve=&SelectFloatFromArray(Direct Array Index=&SpawnGroup(), Float Selection Array=DataInterface() {propertyValues {FloatData [-1,1]}})
    ))
   &ScaleColor(
    Scale Alpha=&FloatFromCurve(FloatCurve=DataInterface() {propertyValues {Curve {keys [
         <at 0 0 RCIM_Cubic>
         {arriveTangent=0.4156, interpMode=RCIM_Cubic, leaveTangent=0.4156, tangentMode=RCTM_User, time=0.05, value=0.9819}
         {arriveTangent=-1.7739, interpMode=RCIM_Cubic, leaveTangent=-1.7739, tangentMode=RCTM_User, time=0.78, value=0.2}
         {arriveTangent=0.1352, interpMode=RCIM_Cubic, leaveTangent=0.1352, tangentMode=RCTM_User, time=0.85, value=0}
        ]}}})
    ScaleA=true
    ScaleRGB=true
   )
   /SetVariables(
    Particles.Scale_SizeBrokenTrail:NiagaraFloat=&Multiply_Float(
     A=$Particles.Size_BrokenTrail
     B=&FloatFromCurve(FloatCurve=DataInterface() {propertyValues {Curve {keys [
          {interpMode=RCIM_Cubic, tangentMode=RCTM_Break, time=0, value=0.3}
          {arriveTangent=-0.05, interpMode=RCIM_Cubic, leaveTangent=-0.05, time=0.5, value=0.45}
          {interpMode=RCIM_Cubic, tangentMode=RCTM_Break, time=1, value=0.25}
         ]}}})
    )
    Particles.Scale_SizeGRadientTrail:NiagaraFloat=&Multiply_Float(
     A=$Particles.Size_GradientTrail
     B=&FloatFromCurve(
      FloatCurve=DataInterface() {propertyValues {Curve {keys [
          {interpMode=RCIM_Cubic, leaveTangent=4.1566, tangentMode=RCTM_Break, time=0, value=0}
          {arriveTangent=0.2467, interpMode=RCIM_Cubic, leaveTangent=0.2467, tangentMode=RCTM_User, time=0.19, value=0.28}
          {arriveTangent=0.2147, interpMode=RCIM_Cubic, leaveTangent=0.5418, tangentMode=RCTM_Break, time=0.25, value=0.32}
          {arriveTangent=-0.2933, interpMode=RCIM_Cubic, leaveTangent=-0.2933, time=0.4297, value=0.35}
          <at 1 0.1 RCIM_Cubic>
         ]}}}
      Scale Curve=&SelectFloatFromArray(Direct Array Index=&SpawnGroup(), Float Selection Array=DataInterface() {propertyValues {FloatData [0.8,-0.8]}})
     )
    )
    Particles.Scale_Sizered:NiagaraFloat=&Multiply_Float(
     A=$Particles.Size_GradientTrail
     B=&FloatFromCurve(FloatCurve=DataInterface() {propertyValues {Curve {keys [
          {interpMode=RCIM_Cubic, leaveTangent=3.8699, tangentMode=RCTM_Break, time=0, value=0}
          {interpMode=RCIM_Cubic, tangentMode=RCTM_Break, time=0.25, value=0.4}
          <at 1.006 0.2 RCIM_Cubic>
         ]}}})
    )
   ) {isSetParameters=true}
   /SetVariables(
    Particles.FADE_DETAILS_COLOR:LinearColor=&ColorFromCurve(ColorCurve=DataInterface() {propertyValues {
       AlphaCurve {keys [<at 0 1 RCIM_Cubic>,<at 0.3008 0>]}
       BlueCurve {keys [<at 0 0 RCIM_Cubic>]}
       GreenCurve {keys [<at 0 0 RCIM_Cubic>]}
       RedCurve {keys [<at 0 0 RCIM_Cubic>]}
      }})
    Particles.ScaleColor_BrokenTraill:LinearColor=&MultiplyLinearColor(
     Linear Color=&MakeLinearColorFromVectorAndFloat(
      "Vector (RGB)"=<0 0 0>
      "Float (Alpha)"=&SelectFloatFromArray(Direct Array Index=&SpawnGroup(), Float Selection Array=DataInterface() {propertyValues {FloatData [0.4,0.55]}})
     )
     Scale Factor=&MakeVector4FromVectorAndFloat(
      "Vector (RGB)"=<1 1 1>
      "Float (Alpha)"=&FloatFromCurve(FloatCurve=DataInterface() {propertyValues {Curve {keys [
           {interpMode=RCIM_Cubic, tangentMode=RCTM_Break, time=0, value=0}
           {
            arriveTangent=0.033
            arriveTangentWeight=0.5583
            interpMode=RCIM_Cubic
            leaveTangent=2.2653
            leaveTangentWeight=0.1102
            tangentMode=RCTM_Break
            time=0.4
            value=1.25
           }
           {arriveTangent=0.1088, interpMode=RCIM_Cubic, tangentMode=RCTM_Break, time=0.75, value=0}
          ]}}})
     )
    )
    Particles.ScaleColor_GradientTrail:LinearColor=&ColorFromCurve(ColorCurve=DataInterface() {propertyValues {
       AlphaCurve {keys [<at 0 1 RCIM_Cubic>,<at 0.7846 0>]}
       BlueCurve {keys [<at 0.0476 0 RCIM_Cubic>,<at 0.1526 0.1562>,<at 0.2411 0>,<at 0.3214 0.0333>,<at 0.449 0.0253>,<at 0.5396 0.0047>,<at 0.5705 0>,<at 0.6117 0>,<at 0.6837 0 RCIM_Cubic>]}
       GreenCurve {keys [<at 0.0476 7.1074 RCIM_Cubic>,<at 0.1526 3.9822>,<at 0.2411 0.4383>,<at 0.3214 0>,<at 0.449 0>,<at 0.5396 0>,<at 0.5705 0>,<at 0.6117 0.15>,<at 0.6837 0 RCIM_Cubic>]}
       RedCurve {keys [<at 0.0476 10 RCIM_Cubic>,<at 0.1526 7.8082>,<at 0.2411 1>,<at 0.3214 1>,<at 0.449 0.7579>,<at 0.5396 0.9552>,<at 0.5705 1>,<at 0.6117 1.8>,<at 0.6837 0 RCIM_Cubic>]}
      }})
   ) {isSetParameters=true}
   &DynamicMaterialParameters(
    Index 1 Param 1=&FloatFromCurve(FloatCurve=DataInterface() {propertyValues {Curve {
        defaultValue=1
        keys [
         <at -0.0073 1 RCIM_Cubic>
         {arriveTangent=-3.2933, interpMode=RCIM_Cubic, leaveTangent=-3.2933, time=0.025, value=-1}
         {arriveTangent=4.4226, interpMode=RCIM_Cubic, leaveTangent=4.4226, tangentMode=RCTM_User, time=0.6, value=-1}
         <at 0.9 2 RCIM_Cubic>
        ]
       }}})
    Index 1 Param 2=1
    Index 2 Param 1=1
    Index 2 Param 2=1
    Index 2 Param 3=&SelectFloatFromArray(Direct Array Index=&SpawnGroup(), Float Selection Array=DataInterface() {propertyValues {FloatData [0.3,-0.3]}})
    Index 2 Param 4=1
    Index 3 Param 1=&SelectFloatFromArray(Direct Array Index=&SpawnGroup(), Float Selection Array=DataInterface() {propertyValues {FloatData [0.3,-0.3]}})
    Index 3 Param 2=1
    Module.Index1Float0Write=true
    Module.Index1Float1Write=true
    Module.Index1Float2Write=true
    Module.Index1Float3Write=true
    Module.Index2Float0Write=true
    Module.Index2Float1Write=true
    Module.Index2Float2Write=true
    Module.Index2Float3Write=true
    Module.Index3Float0Write=true
    Module.Index3Float1Write=true
    Module.Index3Float2Write=true
    Module.Index3Float3Write=true
    Write Parameter Index 0=::NewEnumerator0
    Write Parameter Index 1=::NewEnumerator1
    Write Parameter Index 2=::NewEnumerator1
    Write Parameter Index 3=::NewEnumerator1
   )
  ]
  rendererClasses [&NiagaraRibbonRendererProperties]
  renderers [
   {rendererClass=&NiagaraRibbonRendererProperties}
   {rendererClass=&NiagaraRibbonRendererProperties, rendererIndex=1}
   {rendererClass=&NiagaraRibbonRendererProperties, rendererIndex=2}
  ]
 }
}
summary {
 systemName=NS_Booster_DarkMagic
 emitters [
  {emitterName=Ribbon_TestSharedMaterials, rendererClasses [&NiagaraRibbonRendererProperties]}
  {emitterName=Ribbon_TestSharedMaterials001, rendererClasses [&NiagaraRibbonRendererProperties]}
  {
   emitterName=NE_VC_Boost_Sprite_Parent
   rendererClasses [&NiagaraSpriteRendererProperties, &NiagaraMeshRendererProperties]
  }
 ]
 userVariables [
  User.bIsFrontendPreview {type=NiagaraBool}
  User.bIsPainted {type=NiagaraBool}
  User.ColorVariantID {type=NiagaraInt32}
  User.DTCarOffset {type=Vector3f, description="", defaultValue {struct=Vector3f, value=<0 0 0>}}
  User.GameModeID {type=NiagaraInt32}
  User.PaintedColor {type=LinearColor, description="", defaultValue {struct=LinearColor, value=<rgba 1 1 1 1>}}
  User.PaintedScalar {type=NiagaraFloat, defaultValue=1, description=""}
  User.PortCount {type=NiagaraInt32, defaultValue=2}
 ]
}
